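# Sound cue for one-shot bat ambience inside the Jungle Temple POI, played quietly with wide one-shot attenuation in the indoor ambient sound class.
# A quality-level node sends its first two levels to a random pick of thirteen bat waves and its last two to a smaller six-wave set.
# Each branch randomizes pitch and volume through its own modulator, and the smaller random node is excluded from branch culling.
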SoundCue /Game/Sounds/Artemis/POI/Jungle_Temple/Spots/Spot_Canary_Jungle_Temple_Int_Bats_Cue
root=QualityLevel_0
nodes [
 {id=Modulator_1, children [Random_2], properties {pitchMax=1.1, pitchMin=0.9, volumeMax=1, volumeMin=0.5}}
 {id=Modulator_2, children [Random_3], properties {pitchMax=1.1, volumeMax=1.1, volumeMin=0.1}}
 {id=QualityLevel_0, children [Modulator_1,Modulator_1,Modulator_2,Modulator_2]}
 {
  id=Random_2
  children [
   WavePlayer_16
   WavePlayer_17
   WavePlayer_18
   WavePlayer_19
   WavePlayer_20
   WavePlayer_21
   WavePlayer_22
   WavePlayer_23
   WavePlayer_24
   WavePlayer_25
   WavePlayer_26
   WavePlayer_27
   WavePlayer_28
  ]
  properties {weights [1,1,1,1,1,1,1,1,1,1,1,1,1]}
 }
 {
  id=Random_3
  children [WavePlayer_29, WavePlayer_30, WavePlayer_31, WavePlayer_32, WavePlayer_33, WavePlayer_34]
  properties {bShouldExcludeFromBranchCulling=true, weights [1,1,1,1,1,1]}
 }
 {id=WavePlayer_16, properties {soundWaveAssetPtr=&/AMB_Canary_Jungle_Temple_Int_SS_Bats_01_01}}
 {id=WavePlayer_17, properties {soundWaveAssetPtr=&/AMB_Canary_Jungle_Temple_Int_SS_Bats_01_02}}
 {id=WavePlayer_18, properties {soundWaveAssetPtr=&/AMB_Canary_Jungle_Temple_Int_SS_Bats_01_03}}
 {id=WavePlayer_19, properties {soundWaveAssetPtr=&/AMB_Canary_Jungle_Temple_Int_SS_Bats_01_04}}
 {id=WavePlayer_20, properties {soundWaveAssetPtr=&/AMB_Canary_Jungle_Temple_Int_SS_Bats_01_05}}
 {id=WavePlayer_21, properties {soundWaveAssetPtr=&/AMB_Canary_Jungle_Temple_Int_SS_Bats_01_06}}
 {id=WavePlayer_22, properties {soundWaveAssetPtr=&/AMB_Canary_Jungle_Temple_Int_SS_Bats_02_01}}
 {id=WavePlayer_23, properties {soundWaveAssetPtr=&/AMB_Canary_Jungle_Temple_Int_SS_Bats_02_02}}
 {id=WavePlayer_24, properties {soundWaveAssetPtr=&/AMB_Canary_Jungle_Temple_Int_SS_Bats_02_03}}
 {id=WavePlayer_25, properties {soundWaveAssetPtr=&/AMB_Canary_Jungle_Temple_Int_SS_Bats_02_04}}
 {id=WavePlayer_26, properties {soundWaveAssetPtr=&/AMB_Canary_Jungle_Temple_Int_SS_Bats_02_05}}
 {id=WavePlayer_27, properties {soundWaveAssetPtr=&/AMB_Canary_Jungle_Temple_Int_SS_Bats_02_06}}
 {id=WavePlayer_28, properties {soundWaveAssetPtr=&/AMB_Canary_Jungle_Temple_Int_SS_Bats_02_07}}
 {id=WavePlayer_29, properties {soundWaveAssetPtr=&AMB_Canary_Jungle_Temple_Int_SS_Bats_01_01}}
 {id=WavePlayer_30, properties {soundWaveAssetPtr=&AMB_Canary_Jungle_Temple_Int_SS_Bats_01_02}}
 {id=WavePlayer_31, properties {soundWaveAssetPtr=&AMB_Canary_Jungle_Temple_Int_SS_Bats_01_03}}
 {id=WavePlayer_32, properties {soundWaveAssetPtr=&AMB_Canary_Jungle_Temple_Int_SS_Bats_02_01}}
 {id=WavePlayer_33, properties {soundWaveAssetPtr=&AMB_Canary_Jungle_Temple_Int_SS_Bats_02_02}}
 {id=WavePlayer_34, properties {soundWaveAssetPtr=&AMB_Canary_Jungle_Temple_Int_SS_Bats_02_03}}
]
properties {
 attenuationSettings=/Game/Sounds/Attenuation/Fort_TileAmbient_OneShot_Wider_Att
 bEnableBusSends=false
 maxDistance=8000
 soundClassObject=&Fort_Ambient_Indoors_Athena
 volumeMultiplier=0.2
 preEffectBusSends [{}]
}
