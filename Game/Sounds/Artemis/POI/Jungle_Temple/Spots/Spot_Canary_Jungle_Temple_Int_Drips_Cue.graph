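# Sound cue for one-shot interior water-drip ambience at the Jungle Temple POI, played as indoor ambient sound with one-shot attenuation.
# A quality-level root sends its first two levels to a random pool of seventeen drip waves and its last two to a smaller pool of six.
# Each pool passes through its own modulator for pitch and volume variation, with a much wider volume range on the smaller pool.
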SoundCue /Game/Sounds/Artemis/POI/Jungle_Temple/Spots/Spot_Canary_Jungle_Temple_Int_Drips_Cue
root=QualityLevel_0
nodes [
 {id=Modulator_1, children [Random_0], properties {pitchMax=1.1, pitchMin=0.9, volumeMax=1, volumeMin=0.5}}
 {id=Modulator_2, children [Random_1], properties {pitchMax=1.1, volumeMax=1.1, volumeMin=0.1}}
 {id=QualityLevel_0, children [Modulator_1,Modulator_1,Modulator_2,Modulator_2]}
 {
  id=Random_0
  children [
   WavePlayer_0
   WavePlayer_1
   WavePlayer_2
   WavePlayer_3
   WavePlayer_4
   WavePlayer_5
   WavePlayer_6
   WavePlayer_7
   WavePlayer_8
   WavePlayer_9
   WavePlayer_10
   WavePlayer_11
   WavePlayer_12
   WavePlayer_13
   WavePlayer_14
   WavePlayer_15
   WavePlayer_54
  ]
  properties {weights [1,1,1,1,1,1,1,1,1,1,1,1,1,1,1,1,1]}
 }
 {
  id=Random_1
  children [WavePlayer_55, WavePlayer_56, WavePlayer_57, WavePlayer_58, WavePlayer_59, WavePlayer_60]
  properties {weights [1,1,1,1,1,1]}
 }
 {id=WavePlayer_0, properties {soundWaveAssetPtr=&/AMB_Canary_Jungle_Temple_Int_SS_Drips_02_09}}
 {id=WavePlayer_1, properties {soundWaveAssetPtr=&/AMB_Canary_Jungle_Temple_Int_SS_Drips_01_01}}
 {id=WavePlayer_10, properties {soundWaveAssetPtr=&/AMB_Canary_Jungle_Temple_Int_SS_Drips_02_02}}
 {id=WavePlayer_11, properties {soundWaveAssetPtr=&/AMB_Canary_Jungle_Temple_Int_SS_Drips_02_03}}
 {id=WavePlayer_12, properties {soundWaveAssetPtr=&/AMB_Canary_Jungle_Temple_Int_SS_Drips_02_04}}
 {id=WavePlayer_13, properties {soundWaveAssetPtr=&/AMB_Canary_Jungle_Temple_Int_SS_Drips_02_05}}
 {id=WavePlayer_14, properties {soundWaveAssetPtr=&/AMB_Canary_Jungle_Temple_Int_SS_Drips_02_06}}
 {id=WavePlayer_15, properties {soundWaveAssetPtr=&/AMB_Canary_Jungle_Temple_Int_SS_Drips_02_07}}
 {id=WavePlayer_2, properties {soundWaveAssetPtr=&/AMB_Canary_Jungle_Temple_Int_SS_Drips_01_02}}
 {id=WavePlayer_3, properties {soundWaveAssetPtr=&/AMB_Canary_Jungle_Temple_Int_SS_Drips_01_03}}
 {id=WavePlayer_4, properties {soundWaveAssetPtr=&/AMB_Canary_Jungle_Temple_Int_SS_Drips_01_04}}
 {id=WavePlayer_5, properties {soundWaveAssetPtr=&/AMB_Canary_Jungle_Temple_Int_SS_Drips_01_05}}
 {id=WavePlayer_54, properties {soundWaveAssetPtr=&/AMB_Canary_Jungle_Temple_Int_SS_Drips_02_08}}
 {id=WavePlayer_55, properties {soundWaveAssetPtr=&/AMB_Canary_Jungle_Temple_Int_SS_Drips_01_06}}
 {id=WavePlayer_56, properties {soundWaveAssetPtr=&AMB_Canary_Jungle_Temple_Int_SS_Drips_01_01}}
 {id=WavePlayer_57, properties {soundWaveAssetPtr=&AMB_Canary_Jungle_Temple_Int_SS_Drips_01_02}}
 {id=WavePlayer_58, properties {soundWaveAssetPtr=&AMB_Canary_Jungle_Temple_Int_SS_Drips_01_03}}
 {id=WavePlayer_59, properties {soundWaveAssetPtr=&AMB_Canary_Jungle_Temple_Int_SS_Drips_01_04}}
 {id=WavePlayer_6, properties {soundWaveAssetPtr=&AMB_Canary_Jungle_Temple_Int_SS_Drips_01_06}}
 {id=WavePlayer_60, properties {soundWaveAssetPtr=&AMB_Canary_Jungle_Temple_Int_SS_Drips_01_05}}
 {id=WavePlayer_7, properties {soundWaveAssetPtr=&/AMB_Canary_Jungle_Temple_Int_SS_Drips_01_07}}
 {id=WavePlayer_8, properties {soundWaveAssetPtr=&/AMB_Canary_Jungle_Temple_Int_SS_Drips_01_08}}
 {id=WavePlayer_9, properties {soundWaveAssetPtr=&/AMB_Canary_Jungle_Temple_Int_SS_Drips_02_01}}
]
properties {
 attenuationSettings=/Game/Sounds/Attenuation/Fort_TileAmbient_OneShot_Wider_Att
 bEnableBusSends=false
 maxDistance=8000
 soundClassObject=&Fort_Ambient_Indoors_Athena
 volumeMultiplier=0.501
 preEffectBusSends [{}]
}
