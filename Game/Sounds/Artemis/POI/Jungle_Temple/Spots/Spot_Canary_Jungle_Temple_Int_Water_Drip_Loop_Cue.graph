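# Ambient spot Sound Cue for looping interior water drips at the Jungle Temple POI, classed as indoor ambience.
# A Random node picks one of two equally weighted looping waves, feeding a Modulator for slight upward pitch variation and an Enveloper with a volume dip curve.
# Inline attenuation overrides add occlusion with low-pass filtering and slow interpolation.
# A concurrency override caps simultaneous instances, and bus and submix sends are disabled.
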SoundCue /Game/Sounds/Artemis/POI/Jungle_Temple/Spots/Spot_Canary_Jungle_Temple_Int_Water_Drip_Loop_Cue
root=Enveloper_0
nodes [
 {
  id=Enveloper_0
  children [Modulator_0]
  properties {pitchCurve {editorCurveData {keys [{time=0}]}}, volumeCurve {editorCurveData {keys [<at 0 1 RCIM_Cubic>,<at 0.5 0.5 RCIM_Cubic>,<at 1 1>]}}}
 }
 {id=Modulator_0, children [Random_1], properties {pitchMax=1.1, pitchMin=1}}
 {id=Random_1, children [WavePlayer_1,WavePlayer_3], properties {weights [1,1]}}
 {
  id=WavePlayer_1
  properties {bLooping=true, soundWaveAssetPtr=&/AMB_Canary_Jungle_Temple_Int_Water_Drip_Loop_02}
 }
 {
  id=WavePlayer_3
  properties {bLooping=true, soundWaveAssetPtr=&/AMB_Canary_Jungle_Temple_Int_Water_Drip_Loop_01}
 }
]
properties {
 attenuationSettings=/Game/Athena/Sounds/Props/Abduction_Pipes/Abduction_Pipe_Loop_Att
 bEnableBusSends=false
 bEnableSubmixSends=false
 bOverrideAttenuation=true
 bOverrideConcurrency=true
 maxDistance=2000
 soundClassObject=&Fort_Ambient_Indoors_Athena
 volumeMultiplier=1.002
 attenuationOverrides {
  bEnableOcclusion=true
  falloffDistance=1500
  occlusionInterpolationTime=5
  occlusionLowPassFilterFrequency=1000
  occlusionTraceChannel=ECC_GameTraceChannel11
  occlusionVolumeAttenuation=0.2
  stereoSpread=100
  attenuationShapeExtents=<500 60>
 }
 concurrencyOverrides {maxCount=3}
}
